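# Editor utility list-entry widget for the StyleSheetsConverter tool that shows an object's name on a button and locates its asset in the Content Browser.
# The Set Object event stores the passed object and, after a validity check, writes its display name to the button's text.
# Clicking the button resolves the stored object's Blueprint asset to a path and syncs the Content Browser to it.
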EditorUtilityWidgetBlueprint /StyleSheetsConverter/Utility/StyleSheetsConverterWidgetEntry
graphs [EventGraph {nodes [
   "On Clicked (Button_Browse)"() {is=title K2Node_ComponentBoundEvent}
   > n1 = Get!(execute:Exec) -> (then:Exec, else:Exec, Object:Object Reference)
   > n2 = GetBlueprintAsset(Object=@n1)
   > n3 = GetPathNameForLoadedAsset(LoadedAsset=@n2)
   > SyncBrowserToObjects(AssetPaths=Make Array("[0]":String=@n3) -> (Array:Array of Strings))
   n4 = Set Object!() -> (OutputDelegate, then, NewObject:Object Reference) {is=title K2Node_CustomEvent}
   > n5 = Set Object(Object:Object Reference=@n4.NewObject) -> (Output_Get:Object Reference) {is=title K2Node_VariableSet}
   > Is Valid(InputObject=@n5)
   n6 = Get Object() -> (Object:Object Reference)
   > "SetText (Text)"(self=Get WidgetName_Text() -> (WidgetName_Text:Text Object Reference), InText="To Text (String)"(InString=GetDisplayName(Object=@n6)))
  ]}]
variables [Object {type=Object, default=null}]
widgets {
 info {rootWidgetClass=&EditorUtilityButton, widgetCount=2}
 widgets [
  Button_Browse {widgetClassPath=&EditorUtilityButton, widget=Button_Browse, slot=~}
  WidgetName_Text {
   widgetClassPath=&TextBlock
   bIsVariable=true
   parent=Button_Browse
   widget=WidgetName_Text
   properties {
    text=WBP_SomeWidget
    font {
     bForceMonospaced=false
     bMaterialIsStencil=false
     fontMaterial=null
     fontObject=&Roboto
     letterSpacing=0
     monospacedWidth=1
     size=12
     skewAmount=0
     typefaceFontName=Bold
     outlineSettings {
      bApplyOutlineToDropShadows=false
      bMiteredCorners=false
      bSeparateFillAlpha=false
      outlineMaterial=null
      outlineSize=0
      outlineColor=<rgba 0 0 0 1>
     }
    }
   }
   slot {class=ButtonSlot}
  }
 ]
}
