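# Editor Utility Widget that converts selected Widget Blueprints from manual text block styling to the StyleSheets system using a chosen RuleSet.
# A SinglePropertyView bound to the widget's own RuleSet variable enables the Convert button only when a RuleSet is set.
# Convert passes Content Browser selected Widget Blueprints to ProcessWidgetBlueprints, then lists modified assets as entry widgets in an expandable wrap box.
# SetInfoMessage shows tinted status or error text, hidden when empty.
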EditorUtilityWidgetBlueprint /StyleSheetsConverter/Utility/StyleSheetsConverter
graphs [
 Check for RuleSet Errors {nodes [
   Check for RuleSet Errors() {is=title K2Node_FunctionEntry}
   > n1 = Is Valid(InputObject=Get RuleSet() -> (RuleSet:Style Sheets Conversion Rule Set Object Reference))
   SetConvertButtonEnabled!(execute=@n1.Is Not Valid, self, Enabled:bool=false, TooltipText:Text="LOCTEXT(\"RuleSet not selected!\")") -> (then)
   > SetInfoMessage!(execute, self, Message:Text="LOCTEXT(\"Please select a RuleSet to enable the conversion.\")", IsError:bool=true) -> (then)
   SetConvertButtonEnabled(execute=@n1.Is Valid, Enabled=true, TooltipText="LOCTEXT(\"Click to start conversion.\")")
   > SetInfoMessage(IsError=false)
  ]}
 EventGraph {nodes [
   "On Clicked (Convert_Button)"() {is=title K2Node_ComponentBoundEvent}
   > ClearChildren(self=Get ConvertedWidgets_Box() -> (ConvertedWidgets_Box:Wrap Box Object Reference))
   > n1 = GetSelectedAssetsOfClass(AssetClass=/Script/UMGEditor.WidgetBlueprint)
   > n2 = Branch(Condition=Is Empty!(TargetArray:Array of Object References=@n1) -> (ReturnValue:bool))
   > SetInfoMessage(Message="LOCTEXT(\"No Assets selected. Please, use Content Browser to select widget blueprints to modify.\")", IsError=true)
   SetInfoMessage(execute=@n2.else, IsError=false)
   > n3 = ProcessWidgetBlueprints!(execute, Assets:Array of Object References=@n1, RuleSet:Style Sheets Conversion Rule Set Object Reference=Get RuleSet()) -> (then, OutModifiedWidgets:Array of Object References)
   > n4 = Branch(Condition=Is Empty(TargetArray=@n3))
   > SetInfoMessage(Message="LOCTEXT(\"No assets were modified during the execution.\")", IsError=true)
   n5 = Length(TargetArray:Array of Object References=@n3)
   n6 = "integer > integer"(A=@n5, B=1)
   SetInfoMessage(
    execute=@n4.else
    Message=Format Text!(Format="LOCTEXT(\"Modified {Number} asset{Plural}.\")", Number:int=@n5, Plural:Text=SelectText(A="LOCTEXT(\"s\")", bPickA=@n6)) -> (Result)
    IsError=false
   )
   > n7 = For Each Loop(Array:Array of Object References=@n3) -> (Array Element:Object Reference)
   > n8 = Create Widget!(
    execute:Exec
    WorldContextObject:Object Reference=Self-Reference()
    Class:User Widget Class Reference=&/StyleSheetsConverterWidgetEntry.StyleSheetsConverterWidgetEntry_C
    OwningPlayer:Player Controller Object Reference
   ) -> (then:Exec, ReturnValue:Style Sheets Converter Widget Entry Object Reference)
   > Set Object!(execute, self:Style Sheets Converter Widget Entry Object Reference=@n8, NewObject:Object Reference=@n7) -> (then) {is=title K2Node_CallFunction}
   > AddChildToWrapBox!(execute, self:Wrap Box Object Reference=Get ConvertedWidgets_Box(), Content:Widget Object Reference=@n8) -> (then, ReturnValue:Wrap Box Slot Object Reference)
   SetVisibility(
    execute=@n7.Completed
    self=Get Spoiler_ConvertedWidgets() -> (Spoiler_ConvertedWidgets:Editor Utility Expandable Area Object Reference)
    InVisibility=Visible
   )
   "On Property Changed (RuleSet_PropertyView)"!() -> (OutputDelegate, then, PropertyName:Name=None) {is=title K2Node_ComponentBoundEvent}
   > Check for RuleSet Errors() {is=title K2Node_CallFunction}
   Event PreConstruct()
   > SetObject(self=Get RuleSet_PropertyView() -> (RuleSet_PropertyView:Single Property View Object Reference), NewObject=Self-Reference())
   > Check for RuleSet Errors()
   > SetVisibility(self=Get Spoiler_ConvertedWidgets(), InVisibility=Collapsed)
  ]}
 SetConvertButtonEnabled {nodes [
   n1 = SetConvertButtonEnabled!() -> (then, Enabled:bool=~, TooltipText:Text) {is=title K2Node_FunctionEntry}
   n2 = Get Convert_Button() -> (Convert_Button:Editor Utility Button Object Reference)
   > SetIsEnabled(self=@n2, bInIsEnabled=@n1)
   > SetToolTipText!(execute, self:Widget Object Reference=@n2, InToolTipText:"Text (by ref)"=Get TooltipText() -> (TooltipText:Text)) -> (then)
  ]}
 SetInfoColor {nodes [
   n1 = SetInfoColor!() -> (then, SpecifiedColor:Linear Color Structure) {is=title K2Node_FunctionEntry}
   > SetColorAndOpacity(self:Image Object Reference=Get Warning_Image() -> (Warning_Image:Image Object Reference), InColorAndOpacity:Linear Color Structure=@n1)
   > SetColorAndOpacity(
    self:Text Object Reference=Get InfoMessage_Text() -> (InfoMessage_Text:Text Object Reference)
    InColorAndOpacity:Slate Color Structure=Make Slate Color(SpecifiedColor=@n1)
   )
  ]}
 SetInfoMessage {nodes [
   n1 = SetInfoMessage!() -> (then, Message:Text, IsError:bool=~) {is=title K2Node_FunctionEntry}
   > SetInfoColor!(
    execute
    self
    SpecifiedColor:Linear Color Structure=SelectColor(
     A=Get ErrorMessageTint() -> (ErrorMessageTint:Linear Color Structure)
     B=Get InfoMessageTint() -> (InfoMessageTint:Linear Color Structure)
     bPickA=@n1.IsError
    )
   ) -> (then) {is=title K2Node_CallFunction}
   > n2 = Branch(Condition=TextIsEmpty(InText=@n1))
   > SetVisibility(self=Get NoteMessage_Box() -> (NoteMessage_Box:Horizontal Box Object Reference), InVisibility=Hidden)
   SetVisibility(execute=@n2.else, self=Get NoteMessage_Box(), InVisibility=HitTestInvisible)
   > "SetText (Text)"(self=Get InfoMessage_Text(), InText=Get Message() -> (Message:Text))
  ]}
]
variables [RuleSet {type=StyleSheetsConversionRuleSet, default=null}]
widgets {
 info {namedSlotCount=4, rootWidgetClass=&StackBox, widgetCount=17}
 widgets [
  StackBox_0 {widgetClassPath=&StackBox, widget=StackBox_0, properties {orientation=Orient_Vertical}}
  TextBlock_207 {
   widgetClassPath=&TextBlock
   parent=StackBox_0
   widget=TextBlock_207
   properties {
    text=Style Sheets Converter
    font {
     bForceMonospaced=false
     bMaterialIsStencil=false
     fontMaterial=null
     fontObject=&Roboto
     letterSpacing=0
     monospacedWidth=1
     size=18
     skewAmount=0
     typefaceFontName=Bold
     outlineSettings {
      bApplyOutlineToDropShadows=false
      bMiteredCorners=false
      bSeparateFillAlpha=false
      outlineMaterial=null
      outlineSize=0
      outlineColor=<rgba 0 0 0 1>
     }
    }
   }
   slot {class=StackBoxSlot, properties {horizontalAlignment=HAlign_Center, verticalAlignment=VAlign_Center, padding=<ltrb 8 8 8 8>}}
  }
  HorizontalBox_0 {
   widgetClassPath=&HorizontalBox
   parent=StackBox_0
   widget=HorizontalBox_0
   slot {class=StackBoxSlot, properties {horizontalAlignment=HAlign_Center, padding=<ltrb 32 0 32 0>}}
  }
  Convert_Button {
   widgetClassPath=&EditorUtilityButton
   parent=HorizontalBox_0
   widget=Convert_Button
   slot {class=HorizontalBoxSlot, properties {padding=<ltrb 32 8 32 8>}}
  }
  TextBlock_37 {
   widgetClassPath=&TextBlock
   parent=Convert_Button
   widget=TextBlock_37
   properties {
    text=Convert
    font {
     bForceMonospaced=false
     bMaterialIsStencil=false
     fontMaterial=null
     fontObject=&Roboto
     letterSpacing=0
     monospacedWidth=1
     size=24
     skewAmount=0
     typefaceFontName=Regular
     outlineSettings {
      bApplyOutlineToDropShadows=false
      bMiteredCorners=false
      bSeparateFillAlpha=false
      outlineMaterial=null
      outlineSize=0
      outlineColor=<rgba 0 0 0 1>
     }
    }
   }
   slot {class=ButtonSlot}
  }
  RuleSet_PropertyView {
   widgetClassPath=&SinglePropertyView
   parent=HorizontalBox_0
   widget=RuleSet_PropertyView
   properties {propertyName=RuleSet}
   slot {class=HorizontalBoxSlot, properties {padding=<ltrb 32 8 32 8>}}
  }
  NoteMessage_Box {
   widgetClassPath=&HorizontalBox
   bIsVariable=true
   parent=StackBox_0
   widget=NoteMessage_Box
   slot {class=StackBoxSlot, properties {padding=<ltrb 16 16 16 16>}}
  }
  Warning_Image {
   widgetClassPath=&Image
   bIsVariable=true
   parent=NoteMessage_Box
   widget=Warning_Image
   properties {brush {
     bIsDynamicallyLoaded=false
     drawAs=Image
     imageType=FullColor
     mirroring=NoMirror
     resourceName=None
     resourceObject=/Game/Athena/HUD/Bacchus/HUDLayoutTool/HLT_Icons/T-HLT-Warning-64_Sprite
     tiling=NoTile
     imageSize=<16 16>
     margin=<ltrb 0 0 0 0>
     outlineSettings {
      bUseBrushTransparency=false
      roundingType=HalfHeightRadius
      width=0
      color {colorUseRule=UseColor_Specified, specifiedColor=<rgba 0 0 0 0>}
      cornerRadii=<0 0 0 0>
     }
     tintColor {colorUseRule=UseColor_Specified, specifiedColor=<rgba 1 1 1 1>}
     uVRegion {bIsValid=false, max=<0 0>, min=<0 0>}
    }}
   slot {class=HorizontalBoxSlot, properties {padding=<ltrb 8 8 8 8>}}
  }
  InfoMessage_Text {
   widgetClassPath=&TextBlock
   bIsVariable=true
   parent=NoteMessage_Box
   widget=InfoMessage_Text
   properties {
    text=Text Block
    colorAndOpacity {colorUseRule=UseColor_Specified, specifiedColor=<rgba 1 0.305 0 1>}
    font {
     bForceMonospaced=false
     bMaterialIsStencil=false
     fontMaterial=null
     fontObject=&Roboto
     letterSpacing=0
     monospacedWidth=1
     size=9
     skewAmount=0
     typefaceFontName=Bold
     outlineSettings {
      bApplyOutlineToDropShadows=false
      bMiteredCorners=false
      bSeparateFillAlpha=false
      outlineMaterial=null
      outlineSize=0
      outlineColor=<rgba 0 0 0 1>
     }
    }
   }
   slot {class=HorizontalBoxSlot, properties {verticalAlignment=VAlign_Center, padding=<ltrb 4 4 4 4>}}
  }
  Spoiler_ConvertedWidgets {
   widgetClassPath=/Script/Blutility.EditorUtilityExpandableArea
   bIsVariable=true
   parent=StackBox_0
   widget=Spoiler_ConvertedWidgets
   properties {bIsExpanded=true}
   slot {class=StackBoxSlot, properties {padding=<ltrb 16 0 16 16>}}
  }
  ConvertedWidgetsText {
   widgetClassPath=&TextBlock
   namedSlotHost=Spoiler_ConvertedWidgets
   widget=ConvertedWidgetsText
   properties {
    text=ConvertedWidgets
    font {
     bForceMonospaced=false
     bMaterialIsStencil=false
     fontMaterial=null
     fontObject=&Roboto
     letterSpacing=0
     monospacedWidth=1
     size=9
     skewAmount=0
     typefaceFontName=Regular
     outlineSettings {
      bApplyOutlineToDropShadows=false
      bMiteredCorners=false
      bSeparateFillAlpha=false
      outlineMaterial=null
      outlineSize=0
      outlineColor=<rgba 0 0 0 1>
     }
    }
   }
  }
  Overlay {widgetClassPath=&Overlay, namedSlotHost=Spoiler_ConvertedWidgets, widget=Overlay}
  ConvertedWidgets_Box {
   widgetClassPath=&WrapBox
   bIsVariable=true
   parent=Overlay
   widget=ConvertedWidgets_Box
   slot {class=OverlaySlot, properties {padding=<ltrb 4 4 4 4>}}
  }
  Spoiler_Manual {
   widgetClassPath=&EditorUtilityExpandableArea
   bIsVariable=true
   parent=StackBox_0
   widget=Spoiler_Manual
   slot {class=StackBoxSlot, properties {padding=<ltrb 16 0 16 16>}}
  }
  TextBlock_47 {
   widgetClassPath=&TextBlock
   namedSlotHost=Spoiler_Manual
   widget=TextBlock_47
   properties {
    text=User Manual
    font {
     bForceMonospaced=false
     bMaterialIsStencil=false
     fontMaterial=null
     fontObject=&Roboto
     letterSpacing=0
     monospacedWidth=1
     size=9
     skewAmount=0
     typefaceFontName=Regular
     outlineSettings {
      bApplyOutlineToDropShadows=false
      bMiteredCorners=false
      bSeparateFillAlpha=false
      outlineMaterial=null
      outlineSize=0
      outlineColor=<rgba 0 0 0 1>
     }
    }
   }
  }
  Overlay_57 {widgetClassPath=&Overlay, namedSlotHost=Spoiler_Manual, widget=Overlay_57}
  UserManual_Text {
   widgetClassPath=&TextBlock
   parent=Overlay_57
   widget=UserManual_Text
   properties {
    autoWrapText=true
    text="Welcome to the StyleSheetsConverter tool.\r\nThis tool can be used to convert existing user widgets, that use the old way of defining text block styles (via Font, Typeface, Size, etc) to use the new StyleSheets system.\r\nTo use the tool:\r\n- Create or find an appropriate StyleSheetsConverterRuleSet and set it to the RuleSet field.\r\n- Select one or more widget blueprints that need conversion in the Content Browser.\r\n- Click 'Convert'.\r\n- When the conversion is finished, the list of modified widgets will be created. You can review your changes before saving the widgets by clicking on the widget name in the list and browsing to it inside the content browser..\r\n\r\nThis tool recursively iterates through child widgets inside the widget blueprint and modifies them too! After the conversion al selected widgets that conform to some rule of the ruleset specified will have their style tags assigned."
    font {
     bForceMonospaced=false
     bMaterialIsStencil=false
     fontMaterial=null
     fontObject=&Roboto
     letterSpacing=0
     monospacedWidth=1
     size=9
     skewAmount=0
     typefaceFontName=Regular
     outlineSettings {
      bApplyOutlineToDropShadows=false
      bMiteredCorners=false
      bSeparateFillAlpha=false
      outlineMaterial=null
      outlineSize=0
      outlineColor=<rgba 0 0 0 1>
     }
    }
   }
   slot {class=OverlaySlot, properties {padding=<ltrb 8 8 8 8>}}
  }
 ]
}
